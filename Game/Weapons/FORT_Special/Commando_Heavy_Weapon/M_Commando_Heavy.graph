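# M_Commando_Heavy is a Default Lit material for the Commando heavy weapon, enabled for skeletal meshes and mesh deformers.
# Its graph holds only three scalar constants, with no textures, parameters or visible connections to material outputs.
# It uses legacy pixel depth offset mode.
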Material /Game/Weapons/FORT_Special/Commando_Heavy_Weapon/M_Commando_Heavy
graph {
 properties {bUsedWithMeshDeformer=true, bUsedWithSkeletalMesh=true, pixelDepthOffsetMode=PDOM_Legacy, usedShadingModels=Default Lit}
 nodes [
  Constant() {r=0.4}
  Constant() {r=1}
  Constant() {r=0.4}
 ]
}
