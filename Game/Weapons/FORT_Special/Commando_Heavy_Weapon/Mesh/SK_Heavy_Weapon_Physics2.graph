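# Physics asset for the Commando heavy weapon skeletal mesh, defining its collision bodies for physics simulation.
# One large elongated box on the Handle bone covers the bulk of the weapon, while Root and Handle_Top carry tiny placeholder-sized boxes.
# Two constraints chain Root, Handle and Handle_Top, with swing and twist left free on all axes, so the angular limits have no effect.
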PhysicsAsset /Game/Weapons/FORT_Special/Commando_Heavy_Weapon/Mesh/SK_Heavy_Weapon_Physics2
bodies {
 Handle [{shapeType=Box, extentX=131.8824, extentY=29.9172, extentZ=29.6214, center=<59.7925 0.0205 -1.5552>}]
 Handle_Top [{shapeType=Box, extentX=1.01, extentY=1.01, extentZ=1.01}]
 Root [{shapeType=Box, extentX=1.01, extentY=1.01, extentZ=1.01}]
}
constraints [
 {
  bone1Name=Handle
  bone2Name=Root
  swing1LimitDegrees=45
  swing1Motion=Free
  swing2LimitDegrees=45
  swing2Motion=Free
  twistLimitDegrees=45
  twistMotion=Free
 }
 {
  bone1Name=Handle_Top
  bone2Name=Handle
  swing1LimitDegrees=45
  swing1Motion=Free
  swing2LimitDegrees=45
  swing2Motion=Free
  twistLimitDegrees=45
  twistMotion=Free
 }
]
